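# Physics asset for the chains of the Golden Skeleton cape cosmetic, defining collision bodies and joints for its dynamic bones.
# Two six-link capsule chains (left and right), a two-body middle chain and a large top-cloth box all branch from dyn_base.
# Side links allow moderate swing with little twist, while middle chain and cloth joints are tighter.
# The dyn_base joint to backpack_root leaves swing and twist free.
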PhysicsAsset /BRCosmetics/Accessories/FORT_Capes/Cape_M_MED_The_Golden_Skeleton/Meshes/M_MED_The_Golden_Skeleton_Chains_dyn
bodies {
 backpack_root [{shapeType=Capsule, length=1, radius=1, rotation=<rot 90 0 0>}]
 dyn_base [{shapeType=Capsule, length=1, radius=1, rotation=<rot 90 0 0>}]
 dyn_chains_l_01 [{shapeType=Capsule, length=0.8276, radius=1.0994, center=<1.5 0 0>, rotation=<rot 0 0 90>}]
 dyn_chains_l_02 [{shapeType=Capsule, length=0.8276, radius=1.0994, center=<1.5 0 0>, rotation=<rot 0 0 90>}]
 dyn_chains_l_03 [{shapeType=Capsule, length=0.8276, radius=1.0994, center=<1.5 0 0>, rotation=<rot 0 0 90>}]
 dyn_chains_l_04 [{shapeType=Capsule, length=0.8276, radius=1.0994, center=<1.5 0 0>, rotation=<rot 0 0 90>}]
 dyn_chains_l_05 [{shapeType=Capsule, length=0.8276, radius=1.0994, center=<1.5 0 0>, rotation=<rot 0 0 90>}]
 dyn_chains_l_06 [{shapeType=Capsule, length=0.8276, radius=1.0994, center=<1.5 0 0>, rotation=<rot 0 0 90>}]
 dyn_chains_r_01 [{shapeType=Capsule, length=0.8276, radius=1.0994, center=<1.5 0 0>, rotation=<rot 0 0 90>}]
 dyn_chains_r_02 [{shapeType=Capsule, length=0.8276, radius=1.0994, center=<1.5 0 0>, rotation=<rot 0 0 90>}]
 dyn_chains_r_03 [{shapeType=Capsule, length=0.8276, radius=1.0994, center=<1.5 0 0>, rotation=<rot 0 0 90>}]
 dyn_chains_r_04 [{shapeType=Capsule, length=0.8276, radius=1.0994, center=<1.5 0 0>, rotation=<rot 0 0 90>}]
 dyn_chains_r_05 [{shapeType=Capsule, length=0.8276, radius=1.0994, center=<1.5 0 0>, rotation=<rot 0 0 90>}]
 dyn_chains_r_06 [{shapeType=Capsule, length=0.8276, radius=1.0994, center=<1.5 0 0>, rotation=<rot 0 0 90>}]
 dyn_mid_chains_01 [{shapeType=Capsule, length=2, radius=1, center=<1.7344 0 0.5027>, rotation=<rot 69.9998 0 0>}]
 dyn_mid_chains_02 [{shapeType=Box, extentX=4.7543, extentY=5.1619, extentZ=1.8414, center=<1.9325 0 0>}]
 dyn_top_cloth_mid [{shapeType=Box, extentX=24.9937, extentY=41.1319, extentZ=4.5827, center=<7.6604 0.0002 -2.4609>, rotation=<rot -5 0 0>}]
}
constraints [
 {bone1Name=dyn_chains_l_01, bone2Name=dyn_base, swing1LimitDegrees=45, swing2LimitDegrees=45, twistLimitDegrees=5}
 {bone1Name=dyn_chains_l_02, bone2Name=dyn_chains_l_01, swing1LimitDegrees=45, swing2LimitDegrees=45, twistLimitDegrees=5}
 {bone1Name=dyn_chains_l_03, bone2Name=dyn_chains_l_02, swing1LimitDegrees=45, swing2LimitDegrees=45, twistLimitDegrees=5}
 {bone1Name=dyn_chains_l_04, bone2Name=dyn_chains_l_03, swing1LimitDegrees=45, swing2LimitDegrees=45, twistLimitDegrees=5}
 {bone1Name=dyn_chains_l_05, bone2Name=dyn_chains_l_04, swing1LimitDegrees=45, swing2LimitDegrees=45, twistLimitDegrees=5}
 {bone1Name=dyn_chains_l_06, bone2Name=dyn_chains_l_05, swing1LimitDegrees=45, swing2LimitDegrees=45, twistLimitDegrees=5}
 {bone1Name=dyn_chains_r_01, bone2Name=dyn_base, swing1LimitDegrees=45, swing2LimitDegrees=45, twistLimitDegrees=5}
 {bone1Name=dyn_chains_r_02, bone2Name=dyn_chains_r_01, swing1LimitDegrees=45, swing2LimitDegrees=45, twistLimitDegrees=5}
 {bone1Name=dyn_chains_r_03, bone2Name=dyn_chains_r_02, swing1LimitDegrees=45, swing2LimitDegrees=45, twistLimitDegrees=5}
 {bone1Name=dyn_chains_r_04, bone2Name=dyn_chains_r_03, swing1LimitDegrees=45, swing2LimitDegrees=45, twistLimitDegrees=5}
 {bone1Name=dyn_chains_r_05, bone2Name=dyn_chains_r_04, swing1LimitDegrees=45, swing2LimitDegrees=45, twistLimitDegrees=5}
 {bone1Name=dyn_chains_r_06, bone2Name=dyn_chains_r_05, swing1LimitDegrees=45, swing2LimitDegrees=45, twistLimitDegrees=5}
 {bone1Name=dyn_top_cloth_mid, bone2Name=dyn_base, swing1LimitDegrees=10, swing2LimitDegrees=15, twistLimitDegrees=2}
 {bone1Name=dyn_mid_chains_02, bone2Name=dyn_mid_chains_01, swing1LimitDegrees=10, swing2LimitDegrees=15, twistLimitDegrees=5}
 {bone1Name=dyn_mid_chains_01, bone2Name=dyn_base, swing1LimitDegrees=10, swing2LimitDegrees=15, twistLimitDegrees=5}
 {
  bone1Name=dyn_base
  bone2Name=backpack_root
  swing1LimitDegrees=45
  swing1Motion=Free
  swing2LimitDegrees=45
  swing2Motion=Free
  twistLimitDegrees=45
  twistMotion=Free
 }
]
