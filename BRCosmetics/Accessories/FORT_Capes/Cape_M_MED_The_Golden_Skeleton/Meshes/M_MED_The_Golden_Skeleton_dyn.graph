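# Physics asset for the dynamic bones of the Golden Skeleton cape cosmetic.
# A six-box cape chain, mirrored left and right strap chains of boxes and capsules, and a top cloth box all hang from dyn_base through swing-limited joints.
# Twist is locked on nearly every joint, and each strap root also locks one swing axis.
# An unconstrained dyn_collsion body carries two large capsules, apparently as a collision volume.
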PhysicsAsset /BRCosmetics/Accessories/FORT_Capes/Cape_M_MED_The_Golden_Skeleton/Meshes/M_MED_The_Golden_Skeleton_dyn
bodies {
 dyn_base [{shapeType=Capsule, length=1, radius=1, rotation=<rot 90 0 0>}]
 dyn_cape_01 [{shapeType=Box, extentX=11, extentY=60, extentZ=1.5, center=<5 0 -1.5>}]
 dyn_cape_02 [{shapeType=Box, extentX=10, extentY=60, extentZ=1.5, center=<5 0 -1.5>}]
 dyn_cape_03 [{shapeType=Box, extentX=13, extentY=23.0607, extentZ=2, center=<6 0.0013 -2.5257>}]
 dyn_cape_04 [{shapeType=Box, extentX=12, extentY=20.1215, extentZ=2, center=<6 0 -2>}]
 dyn_cape_05 [{shapeType=Box, extentX=12, extentY=15, extentZ=2, center=<4.5 0 -2>}]
 dyn_cape_06 [{shapeType=Box, extentX=8, extentY=5, extentZ=2, center=<4 0 -2>}]
 dyn_collsion [
  {shapeType=Capsule, length=1.8143, radius=27.5145, center=<-0.0005 30.2217 -21.8023>, rotation=<rot 90 0 90>}
  {shapeType=Capsule, length=0.1758, radius=27.5145, center=<0.0004 26.2075 -44.4717>, rotation=<rot 90 0 90>}
 ]
 dyn_straps_l_01 [{shapeType=Box, extentX=10, extentY=7.5, extentZ=2, center=<5 0 0>, rotation=<rot 0 0 -20>}]
 dyn_straps_l_02 [{shapeType=Box, extentX=10, extentY=7.5, extentZ=2, center=<5 0 0>, rotation=<rot 0 0 -20>}]
 dyn_straps_l_03 [{shapeType=Capsule, length=2.1062, radius=3.7506, center=<6 0 0>, rotation=<rot 90 0 0>}]
 dyn_straps_l_04 [{shapeType=Capsule, length=4.4178, radius=4.5165, center=<6 0 0>, rotation=<rot 90 0 0>}]
 dyn_straps_l_05 [{shapeType=Capsule, length=4.4178, radius=4.5165, center=<6 0 0>, rotation=<rot 90 0 0>}]
 dyn_straps_l_06 [{shapeType=Capsule, length=4.4178, radius=4.5165, center=<6 0 0>, rotation=<rot 90 0 0>}]
 dyn_straps_l_07 [{shapeType=Capsule, length=4.4178, radius=4.5165, center=<6 0 0>, rotation=<rot 90 0 0>}]
 dyn_straps_r_01 [{shapeType=Box, extentX=10, extentY=7.5, extentZ=2, center=<5 0 0>, rotation=<rot 0 0 20>}]
 dyn_straps_r_02 [{shapeType=Box, extentX=10, extentY=7.5, extentZ=2, center=<5 0 0>, rotation=<rot 0 0 20>}]
 dyn_straps_r_03 [{shapeType=Capsule, length=2.1062, radius=3.7657, center=<6 0 0>, rotation=<rot 90 0 0>}]
 dyn_straps_r_04 [{shapeType=Capsule, length=4.4178, radius=4.5165, center=<6 0 0>, rotation=<rot 90 0 0>}]
 dyn_straps_r_05 [{shapeType=Capsule, length=4.4178, radius=4.5165, center=<6 0 0>, rotation=<rot 90 0 0>}]
 dyn_straps_r_06 [{shapeType=Capsule, length=4.4178, radius=4.5165, center=<6 0 0>, rotation=<rot 90 0 0>}]
 dyn_straps_r_07 [{shapeType=Capsule, length=4.4178, radius=4.5165, center=<6 0 0>, rotation=<rot 90 0 0>}]
 dyn_top_cloth_mid [{shapeType=Box, extentX=21.0752, extentY=45.1521, extentZ=3.001, center=<7.9616 0.0005 0.988>, rotation=<rot -5 0 0>}]
}
constraints [
 {bone1Name=dyn_cape_01, bone2Name=dyn_base, swing1LimitDegrees=30, swing2LimitDegrees=15, twistLimitDegrees=5, twistMotion=Locked}
 {bone1Name=dyn_cape_02, bone2Name=dyn_cape_01, swing1LimitDegrees=30, swing2LimitDegrees=15, twistLimitDegrees=5, twistMotion=Locked}
 {bone1Name=dyn_cape_03, bone2Name=dyn_cape_02, swing1LimitDegrees=30, swing2LimitDegrees=15, twistLimitDegrees=1}
 {bone1Name=dyn_cape_04, bone2Name=dyn_cape_03, swing1LimitDegrees=30, swing2LimitDegrees=45, twistLimitDegrees=5, twistMotion=Locked}
 {bone1Name=dyn_cape_05, bone2Name=dyn_cape_04, swing1LimitDegrees=30, swing2LimitDegrees=45, twistLimitDegrees=5, twistMotion=Locked}
 {bone1Name=dyn_cape_06, bone2Name=dyn_cape_05, swing1LimitDegrees=30, swing2LimitDegrees=45, twistLimitDegrees=5, twistMotion=Locked}
 {
  bone1Name=dyn_straps_r_01
  bone2Name=dyn_base
  swing1LimitDegrees=15
  swing2LimitDegrees=45
  swing2Motion=Locked
  twistLimitDegrees=5
  twistMotion=Locked
 }
 {bone1Name=dyn_straps_r_02, bone2Name=dyn_straps_r_01, swing1LimitDegrees=15, swing2LimitDegrees=10, twistLimitDegrees=5, twistMotion=Locked}
 {bone1Name=dyn_straps_r_03, bone2Name=dyn_straps_r_02, swing1LimitDegrees=45, swing2LimitDegrees=45, twistLimitDegrees=5, twistMotion=Locked}
 {bone1Name=dyn_straps_r_04, bone2Name=dyn_straps_r_03, swing1LimitDegrees=45, swing2LimitDegrees=45, twistLimitDegrees=5, twistMotion=Locked}
 {bone1Name=dyn_straps_r_05, bone2Name=dyn_straps_r_04, swing1LimitDegrees=45, swing2LimitDegrees=45, twistLimitDegrees=5, twistMotion=Locked}
 {bone1Name=dyn_straps_r_06, bone2Name=dyn_straps_r_05, swing1LimitDegrees=45, swing2LimitDegrees=45, twistLimitDegrees=5, twistMotion=Locked}
 {bone1Name=dyn_straps_r_07, bone2Name=dyn_straps_r_06, swing1LimitDegrees=45, swing2LimitDegrees=45, twistLimitDegrees=5, twistMotion=Locked}
 {
  bone1Name=dyn_straps_l_01
  bone2Name=dyn_base
  swing1LimitDegrees=15
  swing2LimitDegrees=45
  swing2Motion=Locked
  twistLimitDegrees=5
  twistMotion=Locked
 }
 {bone1Name=dyn_straps_l_02, bone2Name=dyn_straps_l_01, swing1LimitDegrees=15, swing2LimitDegrees=10, twistLimitDegrees=5, twistMotion=Locked}
 {bone1Name=dyn_straps_l_03, bone2Name=dyn_straps_l_02, swing1LimitDegrees=45, swing2LimitDegrees=45, twistLimitDegrees=5, twistMotion=Locked}
 {bone1Name=dyn_straps_l_04, bone2Name=dyn_straps_l_03, swing1LimitDegrees=45, swing2LimitDegrees=45, twistLimitDegrees=5, twistMotion=Locked}
 {bone1Name=dyn_straps_l_05, bone2Name=dyn_straps_l_04, swing1LimitDegrees=45, swing2LimitDegrees=45, twistLimitDegrees=5, twistMotion=Locked}
 {bone1Name=dyn_straps_l_06, bone2Name=dyn_straps_l_05, swing1LimitDegrees=45, swing2LimitDegrees=45, twistLimitDegrees=5, twistMotion=Locked}
 {bone1Name=dyn_straps_l_07, bone2Name=dyn_straps_l_06, swing1LimitDegrees=45, swing2LimitDegrees=45, twistLimitDegrees=5, twistMotion=Locked}
 {bone1Name=dyn_top_cloth_mid, bone2Name=dyn_base, swing1LimitDegrees=10, swing2LimitDegrees=15, twistLimitDegrees=2}
]
