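# Animation Blueprint for the Golden Skeleton cape cosmetic, animating its cape, straps and chains.
# When rigid bodies are enabled, two RigidBody passes simulate it, using a spine-rotated collision bone.
# Otherwise, straps are constrained to cape bones, offset for movement and skydiving, then driven by Trail and Spring controllers.
# EventGraph derives blend alphas, gravity and collision rotation from movement, air state and spine bend; leftover controller and constraint chains are unconnected.
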AnimBlueprint /BRCosmetics/Accessories/FORT_Capes/Cape_M_MED_The_Golden_Skeleton/Meshes/M_MED_The_Golden_Skeleton_Cape_AnimBlueprint
target_skeleton=&/M_MED_The_Golden_Skeleton_Cape_Skeleton
graphs [
 AnimGraph {nodes [
   n1 = Get trail ctrl alpha() -> (trail ctrl alpha:double)
   n2 = Trail controller - Bone: dyn_cape_06!(ComponentPose, Alpha=@n1, FakeVelocity:Vector="0,0,0") -> (Pose) {
    bUsePlanarLimit=true
    chainLength=5
    relaxationSpeedScale=2
    alphaScaleBiasClamp {bInterpResult=true, interpSpeedDecreasing=3}
    planarLimits [{drivingBone {boneName=backpack_root}, planeTransform {location=<0 3 -30>, rotation=<rot 0 0 -85>, scale=<1 1 1>}}]
    rotationLimits [{limitMax=<180 180 180>}, {limitMax=<180 180 180>}, {limitMax=<180 180 180>}, {limitMax=<180 180 180>}, {limitMax=<180 180 180>}]
    rotationOffsets [<0 0 0>, <0 0 0>, <0 0 0>, <0 0 0>, <0 0 0>]
    trailBone {boneName=dyn_cape_06}
    trailRelaxationSpeed {editorCurveData {keys [<at 0 15 RCIM_Cubic>, {interpMode=RCIM_Cubic, leaveTangent=0.5008, time=0.9877, value=10}]}}
   }
   n3 = Spring controller - Bone: dyn_cape_02!(ComponentPose:Component Space Pose Link Structure=@n2, Alpha:float=0) -> (Pose:Component Space Pose Link Structure) {
    is=title AnimGraphNode_SpringBone
    bLimitDisplacement=true
    bRotateX=true
    bRotateY=true
    bRotateZ=true
    bTranslateX=false
    alphaBoolBlend {blendOption=Linear}
    springBone {boneName=dyn_cape_02}
   }
   n4 = Trail controller - Bone: dyn_straps_r_07!(ComponentPose=@n3, Alpha=@n1, FakeVelocity:Vector="0,0,0") -> (Pose) {
    bUsePlanarLimit=true
    chainLength=6
    relaxationSpeedScale=2
    alphaScaleBiasClamp {bInterpResult=true, interpSpeedDecreasing=3}
    planarLimits [{drivingBone {boneName=backpack_root}, planeTransform {location=<0 1.5 -30>, rotation=<rot 0 0 -85>, scale=<1 1 1>}}]
    rotationLimits [
     {limitMax=<180 180 180>}
     {limitMax=<180 180 180>}
     {limitMax=<180 180 180>}
     {limitMax=<180 180 180>}
     {limitMax=<180 180 180>}
     {limitMax=<180 180 180>}
    ]
    rotationOffsets [<0 0 0>, <0 0 0>, <0 0 0>, <0 0 0>, <0 0 0>, <0 0 0>]
    trailBone {boneName=dyn_straps_r_07}
    trailRelaxationSpeed {editorCurveData {keys [<at 0 18 RCIM_Linear>,<at 1 5 RCIM_Linear>]}}
   }
   n5 = Spring controller - Bone: dyn_straps_r_03!(ComponentPose:Component Space Pose Link Structure=@n4, Alpha:float=@n1) -> (Pose:Component Space Pose Link Structure) {
    is=title AnimGraphNode_SpringBone
    bLimitDisplacement=true
    bRotateX=true
    bRotateY=true
    bRotateZ=true
    bTranslateX=false
    alphaBoolBlend {blendOption=Linear}
    springBone {boneName=dyn_straps_r_03}
   }
   n6 = Trail controller - Bone: dyn_straps_l_07!(ComponentPose=@n5, Alpha=@n1, FakeVelocity:Vector="0,0,0") -> (Pose) {
    bUsePlanarLimit=true
    chainLength=6
    relaxationSpeedScale=2
    alphaScaleBiasClamp {bInterpResult=true, interpSpeedDecreasing=3}
    planarLimits [{drivingBone {boneName=backpack_root}, planeTransform {location=<0 1.5 -30>, rotation=<rot 0 0 -85>, scale=<1 1 1>}}]
    rotationLimits [
     {limitMax=<180 180 180>}
     {limitMax=<180 180 180>}
     {limitMax=<180 180 180>}
     {limitMax=<180 180 180>}
     {limitMax=<180 180 180>}
     {limitMax=<180 180 180>}
    ]
    rotationOffsets [<0 0 0>, <0 0 0>, <0 0 0>, <0 0 0>, <0 0 0>, <0 0 0>]
    trailBone {boneName=dyn_straps_l_07}
    trailRelaxationSpeed {editorCurveData {keys [<at 0 18 RCIM_Linear>,<at 1 8 RCIM_Linear>]}}
   }
   n7 = Spring controller - Bone: dyn_straps_l_03!(ComponentPose:Component Space Pose Link Structure=@n6, Alpha:float=@n1) -> (Pose:Component Space Pose Link Structure) {
    is=title AnimGraphNode_SpringBone
    bLimitDisplacement=true
    bRotateX=true
    bRotateY=true
    bRotateZ=true
    bTranslateX=false
    alphaBoolBlend {blendOption=Linear}
    springBone {boneName=dyn_straps_l_03}
   }
   n8 = Trail controller - Bone: dyn_chains_l_06!(ComponentPose=@n7, Alpha=@n1, FakeVelocity:Vector="0,0,0") -> (Pose) {
    bUsePlanarLimit=true
    chainLength=6
    relaxationSpeedScale=2
    alphaScaleBiasClamp {bInterpResult=true, interpSpeedDecreasing=3}
    planarLimits [{drivingBone {boneName=backpack_root}, planeTransform {location=<0 -5.5 0>, rotation=<rot 0 0 -82>, scale=<1 1 1>}}]
    rotationLimits [
     {limitMax=<180 180 180>}
     {limitMax=<180 180 180>}
     {limitMax=<180 180 180>}
     {limitMax=<180 180 180>}
     {limitMax=<180 180 180>}
     {limitMax=<180 180 180>}
    ]
    rotationOffsets [<0 0 0>, <0 0 0>, <0 0 0>, <0 0 0>, <0 0 0>, <0 0 0>]
    trailBone {boneName=dyn_chains_l_06}
    trailRelaxationSpeed {editorCurveData {keys [
       <at 0 20 RCIM_Cubic>
       {arriveTangent=-15, interpMode=RCIM_Cubic, leaveTangent=-15, time=0.4, value=16}
       {arriveTangent=-0.0571, interpMode=RCIM_Cubic, leaveTangent=-0.0571, tangentMode=RCTM_User, time=1, value=5}
      ]}}
   }
   n9 = Spring controller - Bone: dyn_chains_l_01!(ComponentPose:Component Space Pose Link Structure=@n8, Alpha:float=@n1) -> (Pose:Component Space Pose Link Structure) {
    is=title AnimGraphNode_SpringBone
    bLimitDisplacement=true
    bRotateX=true
    bRotateY=true
    bRotateZ=true
    bTranslateX=false
    bTranslateY=false
    bTranslateZ=false
    alphaBoolBlend {blendOption=Linear}
    springBone {boneName=dyn_chains_l_01}
   }
   n10 = Trail controller - Bone: dyn_chains_r_06!(ComponentPose=@n9, Alpha=@n1, FakeVelocity:Vector="0,0,0") -> (Pose) {
    bUsePlanarLimit=true
    chainLength=5
    relaxationSpeedScale=2.5
    alphaScaleBiasClamp {bInterpResult=true, interpSpeedDecreasing=3}
    planarLimits [{drivingBone {boneName=backpack_root}, planeTransform {location=<0 -5.5 0>, rotation=<rot 0 0 -82>, scale=<1 1 1>}}]
    rotationLimits [{limitMax=<180 180 180>}, {limitMax=<180 180 180>}, {limitMax=<180 180 180>}, {limitMax=<180 180 180>}, {limitMax=<180 180 180>}]
    rotationOffsets [<0 0 0>, <0 0 0>, <0 0 0>, <0 0 0>, <0 0 0>]
    trailBone {boneName=dyn_chains_r_06}
    trailRelaxationSpeed {editorCurveData {keys [<at 0 18 RCIM_Cubic>, {arriveTangent=-0.0571, interpMode=RCIM_Cubic, leaveTangent=-0.0571, tangentMode=RCTM_User, time=1, value=5}]}}
   }
   n11 = Spring controller - Bone: dyn_chains_r_01!(ComponentPose:Component Space Pose Link Structure=@n10, Alpha:float=@n1) -> (Pose:Component Space Pose Link Structure) {
    is=title AnimGraphNode_SpringBone
    bLimitDisplacement=true
    bRotateX=true
    bRotateY=true
    bRotateZ=true
    bTranslateX=false
    bTranslateY=false
    bTranslateZ=false
    alphaBoolBlend {blendOption=Linear}
    springBone {boneName=dyn_chains_r_01}
   }
   Component To Local(ComponentPose=@n11)
   Component To Local()
   Component To Local()
   Component To Local()
   n12 = "Constraint - dyn_straps_r_02 (P)"!(ComponentPose, ConstraintWeights_0=1) -> (Pose) {
    alphaBoolBlend {blendOption=Linear}
    boneToModify {boneName=dyn_straps_r_02}
    constraintSetup [{targetBone {boneName=dyn_cape_02}}]
    constraintWeights [1]
   }
   n13 = "Constraint - dyn_straps_l_02 (P)"!(ComponentPose=@n12, ConstraintWeights_0=1) -> (Pose) {
    alphaBoolBlend {blendOption=Linear}
    boneToModify {boneName=dyn_straps_l_02}
    constraintSetup [{targetBone {boneName=dyn_cape_02}}]
    constraintWeights [1]
   }
   n14 = "Constraint - dyn_straps_r_03 (P)"!(ComponentPose=@n13, ConstraintWeights_0=1) -> (Pose) {
    alphaBoolBlend {blendOption=Linear}
    boneToModify {boneName=dyn_straps_r_03}
    constraintSetup [{targetBone {boneName=dyn_cape_03}}]
    constraintWeights [1]
   }
   n15 = "Constraint - dyn_straps_l_03 (P)"!(ComponentPose=@n14, ConstraintWeights_0=1) -> (Pose) {
    alphaBoolBlend {blendOption=Linear}
    boneToModify {boneName=dyn_straps_l_03}
    constraintSetup [{targetBone {boneName=dyn_cape_03}}]
    constraintWeights [1]
   }
   n16 = "Constraint - dyn_straps_r_04 (P)"!(ComponentPose=@n15, ConstraintWeights_0=1) -> (Pose) {
    alphaBoolBlend {blendOption=Linear}
    boneToModify {boneName=dyn_straps_r_04}
    constraintSetup [{targetBone {boneName=dyn_cape_04}}]
    constraintWeights [1]
   }
   n17 = "Constraint - dyn_straps_l_04 (P)"!(ComponentPose=@n16, ConstraintWeights_0=1) -> (Pose) {
    alphaBoolBlend {blendOption=Linear}
    boneToModify {boneName=dyn_straps_l_04}
    constraintSetup [{targetBone {boneName=dyn_cape_04}}]
    constraintWeights [1]
   }
   n18 = "Constraint - dyn_straps_r_05 (P)"!(ComponentPose=@n17, ConstraintWeights_0=1) -> (Pose) {
    alphaBoolBlend {blendOption=Linear}
    boneToModify {boneName=dyn_straps_r_05}
    constraintSetup [{targetBone {boneName=dyn_cape_05}}]
    constraintWeights [1]
   }
   "Constraint - dyn_straps_l_05 (P)"!(ComponentPose=@n18, ConstraintWeights_0=1) -> (Pose) {
    alphaBoolBlend {blendOption=Linear}
    boneToModify {boneName=dyn_straps_l_05}
    constraintSetup [{targetBone {boneName=dyn_cape_05}}]
    constraintWeights [1]
   }
   n19 = "Constraint - dyn_straps_r_02 (P)"(ConstraintWeights_0=1) {
    alphaBoolBlend {blendOption=Linear}
    boneToModify {boneName=dyn_straps_r_02}
    constraintSetup [{targetBone {boneName=dyn_cape_02}}]
    constraintWeights [1]
   }
   n20 = "Constraint - dyn_straps_l_02 (P)"(ComponentPose=@n19, ConstraintWeights_0=1) {
    alphaBoolBlend {blendOption=Linear}
    boneToModify {boneName=dyn_straps_l_02}
    constraintSetup [{targetBone {boneName=dyn_cape_02}}]
    constraintWeights [1]
   }
   n21 = "Constraint - dyn_straps_r_03 (P)"(ComponentPose=@n20, ConstraintWeights_0=1) {
    alphaBoolBlend {blendOption=Linear}
    boneToModify {boneName=dyn_straps_r_03}
    constraintSetup [{targetBone {boneName=dyn_cape_03}}]
    constraintWeights [1]
   }
   n22 = "Constraint - dyn_straps_l_03 (P)"(ComponentPose=@n21, ConstraintWeights_0=1) {
    alphaBoolBlend {blendOption=Linear}
    boneToModify {boneName=dyn_straps_l_03}
    constraintSetup [{targetBone {boneName=dyn_cape_03}}]
    constraintWeights [1]
   }
   n23 = "Constraint - dyn_straps_r_04 (P)"(ComponentPose=@n22, ConstraintWeights_0=1) {
    alphaBoolBlend {blendOption=Linear}
    boneToModify {boneName=dyn_straps_r_04}
    constraintSetup [{targetBone {boneName=dyn_cape_04}}]
    constraintWeights [1]
   }
   n24 = "Constraint - dyn_straps_l_04 (P)"(ComponentPose=@n23, ConstraintWeights_0=1) {
    alphaBoolBlend {blendOption=Linear}
    boneToModify {boneName=dyn_straps_l_04}
    constraintSetup [{targetBone {boneName=dyn_cape_04}}]
    constraintWeights [1]
   }
   n25 = "Constraint - dyn_straps_r_05 (P)"(ComponentPose=@n24, ConstraintWeights_0=1) {
    alphaBoolBlend {blendOption=Linear}
    boneToModify {boneName=dyn_straps_r_05}
    constraintSetup [{targetBone {boneName=dyn_cape_05}}]
    constraintWeights [1]
   }
   "Constraint - dyn_straps_l_05 (P)"(ComponentPose=@n25, ConstraintWeights_0=1) {
    alphaBoolBlend {blendOption=Linear}
    boneToModify {boneName=dyn_straps_l_05}
    constraintSetup [{targetBone {boneName=dyn_cape_05}}]
    constraintWeights [1]
   }
   Local To Component()
   n26 = Local To Component(LocalPose=Use cached pose 'pre pose'())
   n27 = Get override_gravity() -> (override_gravity:Vector)
   n28 = RigidBody!(ComponentPose=Local To Component(LocalPose=Component To Local(ComponentPose=@n26)), Alpha, OverrideWorldGravity:Vector=@n27, ExternalForce) -> (Pose) {
    bClampLinearTranslationLimitToRefPose=true
    bOverrideWorldGravity=true
    overridePhysicsAsset=&/M_MED_The_Golden_Skeleton_dyn
    simulationSpace=BaseBoneSpace
    alphaBoolBlend {blendOption=Linear}
    baseBoneRef {boneName=dyn_base}
    componentAppliedLinearAccClamp=<2000 1500 1500>
    componentLinearAccScale=<0.4 0.3 0.6>
    componentLinearVelScale=<0.4 0.3 0.3>
    overrideWorldGravity {z=-980}
   }
   n29 = RigidBody(ComponentPose=@n28, Alpha=Get physic_alpha() -> (physic_alpha:double), OverrideWorldGravity=@n27) {
    bClampLinearTranslationLimitToRefPose=true
    bOverrideWorldGravity=true
    overridePhysicsAsset=&/M_MED_The_Golden_Skeleton_Chains_dyn
    simulationSpace=BaseBoneSpace
    alphaBoolBlend {blendOption=Linear}
    alphaScaleBiasClamp {bInterpResult=true, interpSpeedDecreasing=5}
    baseBoneRef {boneName=dyn_base}
    componentAppliedLinearAccClamp=<1000 1000 1000>
    componentLinearAccScale=<0.1 0.1 0.1>
    componentLinearVelScale=<0.1 0.1 0.1>
    overrideWorldGravity {z=-980}
   }
   n30 = "Constraint - dyn_straps_r_02 (P)"(ComponentPose=Local To Component(LocalPose=Use cached pose 'pre pose'()), ConstraintWeights_0=1) {
    alphaBoolBlend {blendOption=Linear}
    boneToModify {boneName=dyn_straps_r_02}
    constraintSetup [{targetBone {boneName=dyn_cape_02}}]
    constraintWeights [1]
   }
   n31 = "Constraint - dyn_straps_l_02 (P)"(ComponentPose=@n30, ConstraintWeights_0=1) {
    alphaBoolBlend {blendOption=Linear}
    boneToModify {boneName=dyn_straps_l_02}
    constraintSetup [{targetBone {boneName=dyn_cape_02}}]
    constraintWeights [1]
   }
   n32 = "Constraint - dyn_straps_r_03 (P)"(ComponentPose=@n31, ConstraintWeights_0=1) {
    alphaBoolBlend {blendOption=Linear}
    boneToModify {boneName=dyn_straps_r_03}
    constraintSetup [{targetBone {boneName=dyn_cape_03}}]
    constraintWeights [1]
   }
   n33 = "Constraint - dyn_straps_l_03 (P)"(ComponentPose=@n32, ConstraintWeights_0=1) {
    alphaBoolBlend {blendOption=Linear}
    boneToModify {boneName=dyn_straps_l_03}
    constraintSetup [{targetBone {boneName=dyn_cape_03}}]
    constraintWeights [1]
   }
   n34 = "Constraint - dyn_straps_r_04 (P)"(ComponentPose=@n33, ConstraintWeights_0=1) {
    alphaBoolBlend {blendOption=Linear}
    boneToModify {boneName=dyn_straps_r_04}
    constraintSetup [{targetBone {boneName=dyn_cape_04}}]
    constraintWeights [1]
   }
   n35 = "Constraint - dyn_straps_l_04 (P)"(ComponentPose=@n34, ConstraintWeights_0=1) {
    alphaBoolBlend {blendOption=Linear}
    boneToModify {boneName=dyn_straps_l_04}
    constraintSetup [{targetBone {boneName=dyn_cape_04}}]
    constraintWeights [1]
   }
   n36 = "Constraint - dyn_straps_r_05 (P)"(ComponentPose=@n35, ConstraintWeights_0=1) {
    alphaBoolBlend {blendOption=Linear}
    boneToModify {boneName=dyn_straps_r_05}
    constraintSetup [{targetBone {boneName=dyn_cape_05}}]
    constraintWeights [1]
   }
   n37 = "Constraint - dyn_straps_l_05 (P)"(ComponentPose=@n36, ConstraintWeights_0=1) {
    alphaBoolBlend {blendOption=Linear}
    boneToModify {boneName=dyn_straps_l_05}
    constraintSetup [{targetBone {boneName=dyn_cape_05}}]
    constraintWeights [1]
   }
   n38 = "Transform (Modify) Bone - Bone: dyn_cape_01"(ComponentPose=@n37, Alpha=1, Translation="0,0,-1", Rotation="-6,0.000,0.000", Scale="1,1,1") {
    rotationMode=BMM_Additive
    rotationSpace=BCS_BoneSpace
    translationMode=BMM_Additive
    translationSpace=BCS_BoneSpace
    alphaBoolBlend {blendOption=Linear}
    boneToModify {boneName=dyn_cape_01}
    rotation {pitch=-6}
    translation {z=-1}
   }
   n39 = "Transform (Modify) Bone - Bone: dyn_straps_r_01"(ComponentPose=@n38, Alpha=1, Translation="0,0,0", Rotation="-3,2,0", Scale="1,1,1") {
    rotationMode=BMM_Additive
    rotationSpace=BCS_BoneSpace
    alphaBoolBlend {blendOption=Linear}
    boneToModify {boneName=dyn_straps_r_01}
    rotation {pitch=-3, yaw=2}
   }
   n40 = "Transform (Modify) Bone - Bone: dyn_straps_r_02"(ComponentPose=@n39, Alpha=1, Translation="0,0,0", Rotation="-3,0,0.000", Scale="1,1,1") {
    rotationMode=BMM_Additive
    rotationSpace=BCS_BoneSpace
    alphaBoolBlend {blendOption=Linear}
    boneToModify {boneName=dyn_straps_r_02}
    rotation {pitch=-3}
   }
   n41 = "Transform (Modify) Bone - Bone: dyn_straps_l_01"(ComponentPose=@n40, Alpha=1, Translation="0,0,0", Rotation="-3,-2,0", Scale="1,1,1") {
    rotationMode=BMM_Additive
    rotationSpace=BCS_BoneSpace
    alphaBoolBlend {blendOption=Linear}
    boneToModify {boneName=dyn_straps_l_01}
    rotation {pitch=-3, yaw=-2}
   }
   n42 = "Transform (Modify) Bone - Bone: dyn_straps_l_02"(ComponentPose=@n41, Alpha=1, Translation="0,0,0", Rotation="-3,0,0.000", Scale="1,1,1") {
    rotationMode=BMM_Additive
    rotationSpace=BCS_BoneSpace
    alphaBoolBlend {blendOption=Linear}
    boneToModify {boneName=dyn_straps_l_02}
    rotation {pitch=-3}
   }
   n43 = Get Bwd_Side_Movemetn() -> (Bwd_Side_Movemetn:double)
   n44 = "Transform (Modify) Bone - Bone: dyn_cape_02"!(ComponentPose=@n42, Alpha=@n43, Scale="1,1,1") -> (Pose) {
    rotationMode=BMM_Additive
    rotationSpace=BCS_BoneSpace
    translationMode=BMM_Additive
    translationSpace=BCS_BoneSpace
    alphaBoolBlend {blendOption=Linear}
    alphaScaleBiasClamp {bInterpResult=true}
    boneToModify {boneName=dyn_cape_02}
    rotation {pitch=15}
   }
   n45 = "Transform (Modify) Bone - Bone: dyn_straps_r_01"!(ComponentPose=@n44, Alpha=@n43, Scale="1,1,1") -> (Pose) {
    rotationMode=BMM_Additive
    rotationSpace=BCS_BoneSpace
    translationMode=BMM_Additive
    translationSpace=BCS_BoneSpace
    alphaBoolBlend {blendOption=Linear}
    alphaScaleBiasClamp {bInterpResult=true}
    boneToModify {boneName=dyn_straps_r_01}
   }
   n46 = "Transform (Modify) Bone - Bone: dyn_straps_r_02"!(ComponentPose=@n45, Alpha=@n43, Scale="1,1,1") -> (Pose) {
    rotationMode=BMM_Additive
    rotationSpace=BCS_BoneSpace
    translationMode=BMM_Additive
    translationSpace=BCS_BoneSpace
    alphaBoolBlend {blendOption=Linear}
    alphaScaleBiasClamp {bInterpResult=true}
    boneToModify {boneName=dyn_straps_r_02}
    rotation {pitch=10}
   }
   n47 = "Transform (Modify) Bone - Bone: dyn_straps_l_01"!(ComponentPose=@n46, Alpha=@n43, Scale="1,1,1") -> (Pose) {
    rotationMode=BMM_Additive
    rotationSpace=BCS_BoneSpace
    translationMode=BMM_Additive
    translationSpace=BCS_BoneSpace
    alphaBoolBlend {blendOption=Linear}
    alphaScaleBiasClamp {bInterpResult=true}
    boneToModify {boneName=dyn_straps_l_01}
   }
   n48 = "Transform (Modify) Bone - Bone: dyn_straps_l_02"!(ComponentPose=@n47, Alpha=@n43, Scale="1,1,1") -> (Pose) {
    rotationMode=BMM_Additive
    rotationSpace=BCS_BoneSpace
    translationMode=BMM_Additive
    translationSpace=BCS_BoneSpace
    alphaBoolBlend {blendOption=Linear}
    alphaScaleBiasClamp {bInterpResult=true}
    boneToModify {boneName=dyn_straps_l_02}
    rotation {pitch=10}
   }
   n49 = Get bIsSkyDiving()
   n50 = "Transform (Modify) Bone - Bone: backpack_root"!(ComponentPose=@n48, bAlphaBoolEnabled:bool=@n49, Scale="1,1,1") -> (Pose) {
    alphaInputType=Bool
    rotationMode=BMM_Additive
    rotationSpace=BCS_BoneSpace
    translationMode=BMM_Additive
    translationSpace=BCS_BoneSpace
    alphaBoolBlend {blendOption=Linear}
    boneToModify {boneName=backpack_root}
    rotation=<rot 6 0 3.9513>
    translation=<0 -1.3437 0>
   }
   n51 = "Transform (Modify) Bone - Bone: dyn_cape_03"!(ComponentPose=@n50, bAlphaBoolEnabled:bool=@n49, Scale="1,1,1") -> (Pose) {
    alphaInputType=Bool
    rotationMode=BMM_Additive
    rotationSpace=BCS_BoneSpace
    translationMode=BMM_Additive
    translationSpace=BCS_BoneSpace
    alphaBoolBlend {blendOption=Linear}
    boneToModify {boneName=dyn_cape_03}
    rotation {pitch=6}
    translation=<-0.2606 0 4.2864>
   }
   n52 = "Transform (Modify) Bone - Bone: dyn_cape_04"!(ComponentPose=@n51, bAlphaBoolEnabled:bool=@n49, Scale="1,1,1") -> (Pose) {
    alphaInputType=Bool
    rotationMode=BMM_Additive
    rotationSpace=BCS_BoneSpace
    translationMode=BMM_Additive
    translationSpace=BCS_BoneSpace
    alphaBoolBlend {blendOption=Linear}
    boneToModify {boneName=dyn_cape_04}
    rotation {pitch=6}
    translation {z=2}
   }
   n53 = "Transform (Modify) Bone - Bone: dyn_cape_05"!(ComponentPose=@n52, bAlphaBoolEnabled:bool=@n49, Scale="1,1,1") -> (Pose) {
    alphaInputType=Bool
    rotationMode=BMM_Additive
    rotationSpace=BCS_BoneSpace
    translationMode=BMM_Additive
    translationSpace=BCS_BoneSpace
    alphaBoolBlend {blendOption=Linear}
    boneToModify {boneName=dyn_cape_05}
    rotation {pitch=6}
    translation {z=2}
   }
   n54 = "Transform (Modify) Bone - Bone: dyn_straps_r_02"!(ComponentPose=@n53, bAlphaBoolEnabled:bool=@n49, Scale="1,1,1") -> (Pose) {
    alphaInputType=Bool
    rotationMode=BMM_Additive
    rotationSpace=BCS_BoneSpace
    translationMode=BMM_Additive
    translationSpace=BCS_BoneSpace
    alphaBoolBlend {blendOption=Linear}
    boneToModify {boneName=dyn_straps_r_02}
    rotation {pitch=5, yaw=1}
    translation {z=1}
   }
   n55 = "Transform (Modify) Bone - Bone: dyn_straps_r_03"!(ComponentPose=@n54, bAlphaBoolEnabled:bool=@n49, Scale="1,1,1") -> (Pose) {
    alphaInputType=Bool
    rotationMode=BMM_Additive
    rotationSpace=BCS_BoneSpace
    translationMode=BMM_Additive
    translationSpace=BCS_BoneSpace
    alphaBoolBlend {blendOption=Linear}
    boneToModify {boneName=dyn_straps_r_03}
    rotation {pitch=5, yaw=1}
    translation {z=1}
   }
   n56 = "Transform (Modify) Bone - Bone: dyn_straps_r_04"!(ComponentPose=@n55, bAlphaBoolEnabled:bool=@n49, Scale="1,1,1") -> (Pose) {
    alphaInputType=Bool
    rotationMode=BMM_Additive
    rotationSpace=BCS_BoneSpace
    translationMode=BMM_Additive
    translationSpace=BCS_BoneSpace
    alphaBoolBlend {blendOption=Linear}
    boneToModify {boneName=dyn_straps_r_04}
    rotation {pitch=5, yaw=1}
    translation {z=1}
   }
   n57 = "Transform (Modify) Bone - Bone: dyn_straps_r_05"!(ComponentPose=@n56, bAlphaBoolEnabled:bool=@n49, Scale="1,1,1") -> (Pose) {
    alphaInputType=Bool
    rotationMode=BMM_Additive
    rotationSpace=BCS_BoneSpace
    translationMode=BMM_Additive
    translationSpace=BCS_BoneSpace
    alphaBoolBlend {blendOption=Linear}
    boneToModify {boneName=dyn_straps_r_05}
    rotation {pitch=5, yaw=1}
    translation {z=1}
   }
   n58 = "Transform (Modify) Bone - Bone: dyn_straps_l_02"!(ComponentPose=@n57, bAlphaBoolEnabled:bool=@n49, Scale="1,1,1") -> (Pose) {
    alphaInputType=Bool
    rotationMode=BMM_Additive
    rotationSpace=BCS_BoneSpace
    translationMode=BMM_Additive
    translationSpace=BCS_BoneSpace
    alphaBoolBlend {blendOption=Linear}
    boneToModify {boneName=dyn_straps_l_02}
    rotation {pitch=5, yaw=-1}
    translation {z=1}
   }
   n59 = "Transform (Modify) Bone - Bone: dyn_straps_l_03"!(ComponentPose=@n58, bAlphaBoolEnabled:bool=@n49, Scale="1,1,1") -> (Pose) {
    alphaInputType=Bool
    rotationMode=BMM_Additive
    rotationSpace=BCS_BoneSpace
    translationMode=BMM_Additive
    translationSpace=BCS_BoneSpace
    alphaBoolBlend {blendOption=Linear}
    boneToModify {boneName=dyn_straps_l_03}
    rotation {pitch=5, yaw=-1}
    translation {z=1}
   }
   n60 = "Transform (Modify) Bone - Bone: dyn_straps_l_04"!(ComponentPose=@n59, bAlphaBoolEnabled:bool=@n49, Scale="1,1,1") -> (Pose) {
    alphaInputType=Bool
    rotationMode=BMM_Additive
    rotationSpace=BCS_BoneSpace
    translationMode=BMM_Additive
    translationSpace=BCS_BoneSpace
    alphaBoolBlend {blendOption=Linear}
    boneToModify {boneName=dyn_straps_l_04}
    rotation {pitch=5, yaw=-1}
    translation {z=1}
   }
   n61 = "Transform (Modify) Bone - Bone: dyn_straps_l_05"!(ComponentPose=@n60, bAlphaBoolEnabled:bool=@n49, Scale="1,1,1") -> (Pose) {
    alphaInputType=Bool
    rotationMode=BMM_Additive
    rotationSpace=BCS_BoneSpace
    translationMode=BMM_Additive
    translationSpace=BCS_BoneSpace
    alphaBoolBlend {blendOption=Linear}
    boneToModify {boneName=dyn_straps_l_05}
    rotation {pitch=5, yaw=-1}
    translation {z=1}
   }
   n62 = Get BwdMovement() -> (BwdMovement:double)
   n63 = Trail controller - Bone: dyn_cape_06(ComponentPose=@n61, Alpha=@n62, FakeVelocity="0,0,0") {
    bUsePlanarLimit=true
    chainLength=4
    relaxationSpeedScale=2
    alphaScaleBiasClamp {bInterpResult=true, interpSpeedDecreasing=3}
    planarLimits [{drivingBone {boneName=backpack_root}, planeTransform {location=<0 3 -30>, rotation=<rot 0 0 -85>, scale=<1 1 1>}}]
    rotationLimits [{limitMax=<180 180 180>}, {limitMax=<180 180 180>}, {limitMax=<180 180 180>}, {limitMax=<180 180 180>}]
    rotationOffsets [<0 0 0>, <0 0 0>, <0 0 0>, <0 0 0>]
    trailBone {boneName=dyn_cape_06}
    trailRelaxationSpeed {editorCurveData {keys [<at 0 15 RCIM_Cubic>, {interpMode=RCIM_Cubic, leaveTangent=0.5008, time=0.9877, value=10}]}}
   }
   n64 = Spring controller - Bone: dyn_cape_02(ComponentPose=@n63, Alpha=@n62) {
    bLimitDisplacement=true
    bRotateX=true
    bRotateY=true
    bRotateZ=true
    bTranslateX=false
    alphaBoolBlend {blendOption=Linear}
    springBone {boneName=dyn_cape_02}
   }
   n65 = Trail controller - Bone: dyn_straps_r_07(ComponentPose=@n64, Alpha=@n62, FakeVelocity="0,0,0") {
    bUsePlanarLimit=true
    chainLength=5
    relaxationSpeedScale=2
    alphaScaleBiasClamp {bInterpResult=true, interpSpeedDecreasing=3}
    planarLimits [{drivingBone {boneName=backpack_root}, planeTransform {location=<0 1.5 -30>, rotation=<rot 0 0 -85>, scale=<1 1 1>}}]
    rotationLimits [{limitMax=<180 180 180>}, {limitMax=<180 180 180>}, {limitMax=<180 180 180>}, {limitMax=<180 180 180>}, {limitMax=<180 180 180>}]
    rotationOffsets [<0 0 0>, <0 0 0>, <0 0 0>, <0 0 0>, <0 0 0>]
    trailBone {boneName=dyn_straps_r_07}
    trailRelaxationSpeed {editorCurveData {keys [<at 0 18 RCIM_Linear>,<at 1 5 RCIM_Linear>]}}
   }
   n66 = Spring controller - Bone: dyn_straps_r_03(ComponentPose=@n65, Alpha=@n62) {
    bLimitDisplacement=true
    bRotateX=true
    bRotateY=true
    bRotateZ=true
    bTranslateX=false
    alphaBoolBlend {blendOption=Linear}
    springBone {boneName=dyn_straps_r_03}
   }
   n67 = Trail controller - Bone: dyn_straps_l_07(ComponentPose=@n66, Alpha=@n62, FakeVelocity="0,0,0") {
    bUsePlanarLimit=true
    chainLength=5
    relaxationSpeedScale=2
    alphaScaleBiasClamp {bInterpResult=true, interpSpeedDecreasing=3}
    planarLimits [{drivingBone {boneName=backpack_root}, planeTransform {location=<0 1.5 -30>, rotation=<rot 0 0 -85>, scale=<1 1 1>}}]
    rotationLimits [{limitMax=<180 180 180>}, {limitMax=<180 180 180>}, {limitMax=<180 180 180>}, {limitMax=<180 180 180>}, {limitMax=<180 180 180>}]
    rotationOffsets [<0 0 0>, <0 0 0>, <0 0 0>, <0 0 0>, <0 0 0>]
    trailBone {boneName=dyn_straps_l_07}
    trailRelaxationSpeed {editorCurveData {keys [<at 0 18 RCIM_Linear>,<at 1 8 RCIM_Linear>]}}
   }
   n68 = Spring controller - Bone: dyn_straps_l_03(ComponentPose=@n67, Alpha=@n62) {
    bLimitDisplacement=true
    bRotateX=true
    bRotateY=true
    bRotateZ=true
    bTranslateX=false
    alphaBoolBlend {blendOption=Linear}
    springBone {boneName=dyn_straps_l_03}
   }
   n69 = Trail controller - Bone: dyn_chains_l_06(ComponentPose=@n68, Alpha=@n62, FakeVelocity="0,0,0") {
    bUsePlanarLimit=true
    chainLength=5
    relaxationSpeedScale=2
    alphaScaleBiasClamp {bInterpResult=true, interpSpeedDecreasing=3}
    planarLimits [{drivingBone {boneName=backpack_root}, planeTransform {location=<0 -5.5 0>, rotation=<rot 0 0 -82>, scale=<1 1 1>}}]
    rotationLimits [{limitMax=<180 180 180>}, {limitMax=<180 180 180>}, {limitMax=<180 180 180>}, {limitMax=<180 180 180>}, {limitMax=<180 180 180>}]
    rotationOffsets [<0 0 0>, <0 0 0>, <0 0 0>, <0 0 0>, <0 0 0>]
    trailBone {boneName=dyn_chains_l_06}
    trailRelaxationSpeed {editorCurveData {keys [
       <at 0 20 RCIM_Cubic>
       {arriveTangent=-15, interpMode=RCIM_Cubic, leaveTangent=-15, time=0.4, value=16}
       {arriveTangent=-0.0571, interpMode=RCIM_Cubic, leaveTangent=-0.0571, tangentMode=RCTM_User, time=1, value=5}
      ]}}
   }
   n70 = Spring controller - Bone: dyn_chains_l_01(ComponentPose=@n69, Alpha=@n62) {
    bLimitDisplacement=true
    bRotateX=true
    bRotateY=true
    bRotateZ=true
    bTranslateX=false
    bTranslateY=false
    bTranslateZ=false
    alphaBoolBlend {blendOption=Linear}
    springBone {boneName=dyn_chains_l_01}
   }
   n71 = Trail controller - Bone: dyn_chains_r_06(ComponentPose=@n70, Alpha=@n62, FakeVelocity="0,0,0") {
    bUsePlanarLimit=true
    chainLength=5
    relaxationSpeedScale=2.5
    alphaScaleBiasClamp {bInterpResult=true, interpSpeedDecreasing=3}
    planarLimits [{drivingBone {boneName=backpack_root}, planeTransform {location=<0 -5.5 0>, rotation=<rot 0 0 -82>, scale=<1 1 1>}}]
    rotationLimits [{limitMax=<180 180 180>}, {limitMax=<180 180 180>}, {limitMax=<180 180 180>}, {limitMax=<180 180 180>}, {limitMax=<180 180 180>}]
    rotationOffsets [<0 0 0>, <0 0 0>, <0 0 0>, <0 0 0>, <0 0 0>]
    trailBone {boneName=dyn_chains_r_06}
    trailRelaxationSpeed {editorCurveData {keys [<at 0 18 RCIM_Cubic>, {arriveTangent=-0.0571, interpMode=RCIM_Cubic, leaveTangent=-0.0571, tangentMode=RCTM_User, time=1, value=5}]}}
   }
   n72 = Spring controller - Bone: dyn_chains_r_01(ComponentPose=@n71, Alpha=@n62) {
    bLimitDisplacement=true
    bRotateX=true
    bRotateY=true
    bRotateZ=true
    bTranslateX=false
    bTranslateY=false
    bTranslateZ=false
    alphaBoolBlend {blendOption=Linear}
    springBone {boneName=dyn_chains_r_01}
   }
   Output Pose(Result=Blend Poses by bool(BlendPose_0=Component To Local(ComponentPose=@n29), BlendPose_1=Component To Local(ComponentPose=@n72), bActiveValue=Get bRigidBodyEnabled()) {blendType=Linear}) {layerGroup=None}
   n73 = "Transform (Modify) Bone - Bone: dyn_collsion"(
    ComponentPose=Local To Component(LocalPose=Input Pose() {name=pre pose})
    Alpha=1
    Translation="0,0,0"
    Rotation=Get backpack_rotate()
    Scale="1,1,1"
   ) {rotationMode=BMM_Additive, rotationSpace=BCS_BoneSpace, alphaBoolBlend {blendOption=Linear}, boneToModify {boneName=dyn_collsion}}
   Save cached pose 'pre pose'(Pose=Component To Local(ComponentPose=@n73)) {cacheName=pre pose}
  ]}
 EventGraph {nodes [
   Event BlueprintUpdateAnimation()
   > n1 = Sequence!(execute) -> (then_0, then_1, then_2:Exec)
   > n2 = PlayerMovement(AnimInstance=Self-Reference(), XDotProductCoefficient=1, YDotProductCoefficient=1, ZDotProductCoefficient=1)
   n3 = "Absolute (Float)"(A=@n2)
   > Set BwdMovement(BwdMovement:double=float * float(
     A=MapRangeClamped(Value=@n3, InRangeA=0, InRangeB=1, OutRangeA=1, OutRangeB=0.6)
     B=MapRangeClamped(Value=@n2.YMovement, InRangeA=0, InRangeB=-1, OutRangeA=0.7, OutRangeB=0)
    )) -> (Output_Get:double)
   n4 = MapRangeClamped(Value=@n2.YMovement, InRangeA=0, InRangeB=-1)
   n5 = MapRangeClamped(Value=@n3, InRangeA=0, InRangeB=1)
   > Set Bwd_Side_Movemetn(Bwd_Side_Movemetn:double="Max (Float)"(A=@n5, B=@n4)) -> (Output_Get:double)
   n6 = Branch(execute=@n1.then_1, Condition=Get bIsSkyDiving())
   > Set Bwd_Side_Movemetn(Bwd_Side_Movemetn=0)
   n7 = Get PawnVelocityZ()
   > n8 = Branch(Condition=OR Boolean(
     A="float < float"!(A:double=@n7, B:double=-1200) -> (ReturnValue:bool) {is=title K2Node_CallFunction}
     B="float > float"(A=@n7, B=50) {is=title K2Node_CallFunction}
    ))
   > Set BwdMovement(BwdMovement=float * float(A=Get BwdMovement(), B=0))
   Set BwdMovement(execute=[@n6.else, @n8.else], BwdMovement=float * float(A=Get BwdMovement(), B=1))
   n9 = OR Boolean!(A=Get bIsJumping(), B=Get bIsFalling(), C:bool=Get bIsSkyDiving()) -> (ReturnValue)
   Set trail ctrl alpha(execute=@n1.then_2, trail ctrl alpha:double=SelectFloat(A=0.3, B=1, bPickA=@n9)) -> (Output_Get:double)
   > Set physic_alpha(physic_alpha:double=SelectFloat(A=0.4, B=1, bPickA=@n9)) -> (Output_Get:double)
   > Set override_gravity(override_gravity:Vector=SelectVector(A="0, 0,-300", B="0, 0,-980", bPickA=Get bIsSkydiveFloating())) -> (Output_Get:Vector)
   n10 = GetOwnerSkeletalMeshComponent()
   n11 = BreakRotator(InRot=BreakTransform(InTransform=GetDeltaTransformFromRefPose(self=@n10, BoneName=spine_05, BaseName=spine_01)).Rotation)
   n12 = MapRangeClamped(Value=@n11.Yaw, InRangeA=5, InRangeB=15, OutRangeB=0)
   n13 = MapRangeClamped(Value=@n11.Yaw, InRangeA=0, InRangeB=-30, OutRangeB=40)
   > Set backpack_rotate(backpack_rotate=MakeRotator(Roll=float + float(A=@n13, B=@n12) {is=title K2Node_CommutativeAssociativeBinaryOperator}))
   TryGetPawnOwner()
   float * float(A=@n5, B=@n4)
  ]}
]
variables [
 backpack_rotate {type=Rotator}
 Bwd_Side_Movemetn {type=number}
 BwdMovement {type=number}
 override_gravity {type=Vector}
 physic_alpha {type=number}
 trail ctrl alpha {type=number, default=1}
]
